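; SSI_BpLib_BpTools is a Blueprint helper library in the snow and sand interaction effects folder, providing an OnLocalPlayersTeam check for a target actor.
; The function starts TeamValid as false, loops over the local Fortnite player controllers and sets it true when any shares a team with the target.
; It returns that flag, so one matching local player is enough.
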(fn OnLocalPlayersTeam
  (params (TargetActor :type Actor%20Object%20Reference))
  (LocalVariables|SetTeamValid false)
  (for _array_element (Fortnite|GetLocalFortPlayerControllers true)
    (if (Team|OnSameTeam _array_element TargetActor)
      (LocalVariables|SetTeamValid true)))
  (return (LocalVariables|GetTeamValid)))
